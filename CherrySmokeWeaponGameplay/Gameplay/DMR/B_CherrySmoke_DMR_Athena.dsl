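; CherrySmoke DMR weapon Blueprint for Athena that decides which shots fire an unstable bullet.
; ShouldFireUnstableBullet derives shot position from magazine size and remaining ammo, checking it against the scalable ProcPerBullets interval.
; A first-shot flag offsets that cycle so the magazine's opening shot qualifies.
; With infinite magazine ammo, a FiredCount counter replaces the magazine math, triggering at zero and wrapping at the interval.
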(fn ShouldFireUnstableBullet
  (params (bIsFirstShotUnstable :type Boolean))
  (LocalVariables|SetMaxClip (FortWeapon|GetBulletsPerClip))
  (LocalVariables|SetCurrentClip (FortWeapon|GetMagazineAmmoCount))
  (if (FortWeapon|Ammo|HasInfiniteMagazineAmmo)
    (LocalVariables|SetbDoUnstable (== (Variables|Default|GetFiredCount) 0))
    (bind _result (Math|Integer|IncrementInt (Variables|Default|GetFiredCount))
      (:%20%20
        (if (>= _result (ScalableFloat|GetValueasInteger (Variables|Default|GetProcPerBullets)))
          (Variables|Default|SetFiredCount 0)
          (else
            _))
        (return (LocalVariables|GetbDoUnstable))))
    (elif (Variables|GetbIsFirstShotUnstable)
      (if (== (LocalVariables|GetMaxClip) (+ (LocalVariables|GetCurrentClip) 1))
        (return true)
        (else
          (return (== (Math|Float|Division(WholeandRemainder) (Math|Conversions|ToFloat(Integer) (- (- (LocalVariables|GetMaxClip) (LocalVariables|GetCurrentClip)) 1)) (Math|Conversions|ToFloat(Integer) (ScalableFloat|GetValueasInteger (Variables|Default|GetProcPerBullets)))) 0))))
      (else
        (return (== (Math|Float|Division(WholeandRemainder) (Math|Conversions|ToFloat(Integer) (- (LocalVariables|GetMaxClip) (LocalVariables|GetCurrentClip))) (Math|Conversions|ToFloat(Integer) (ScalableFloat|GetValueasInteger (Variables|Default|GetProcPerBullets)))) 0))))))

(fn ConstructionScript
  (|CallParentFunction))
